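# DiceGame is a short Level Sequence that animates a board game turn with two static mesh props, a die and a player piece.
# The die tumbles and bounces over roughly the first second through keyed location and rotation.
# The piece then advances in five hops, made by alternating height keys, while yawing to change direction.
# One audio track spans the whole animation at a 30 fps display rate.
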LevelSequence /MiscReferencersOfCosmetics/DiceGame
bindings [
 SM_Int_Boardgame_PlayerPiece_AG_FE {
  id=#1
  possessed_class=FortStaticMeshActor
  tracks [{
    type=MovieScene3DTransformTrack
    sections [{
      type=MovieScene3DTransformSection
      channels [
       Location.X {
        type=MovieSceneScriptingDoubleChannel
        default=2307.604
        thinned_from=13
        keys [[29.38,2307.604],[44.6888,2316.7966],[84.6888,2287.123],[108.6888,2275.8738]]
       }
       Location.Y {type=MovieSceneScriptingDoubleChannel, default=-106927.625, thinned_from=13, keys [[29.38,-106927.625],[108.6888,-106902.0469]]}
       Location.Z {
        type=MovieSceneScriptingDoubleChannel
        default=14.504
        keys [[29.38,14.504],[31.6888,14.504],[38.6888,17.2623],[44.6888,14.504],[52.6888,17.5685],[58.6888,14.504],[68.6888,18.2943],[76.6888,14.504],[80.6888,16.4811],[84.6888,14.504],[91.6888,18.444],[100.6888,14.3816],[108.6888,14.3816]]
       }
       Rotation.X {type=MovieSceneScriptingDoubleChannel, default=0, const [0,5]}
       Rotation.Y {type=MovieSceneScriptingDoubleChannel, default=0, const [0,5]}
       Rotation.Z {
        type=MovieSceneScriptingDoubleChannel
        default=-30.0003
        keys [[29.38,-100.0002],[31.6888,-100.0002],[44.6888,19.9998],[84.6888,-150.0003],[108.6888,-150.0003]]
       }
       Scale.X {type=MovieSceneScriptingDoubleChannel, default=1.25, const [1.25,13]}
       Scale.Y {type=MovieSceneScriptingDoubleChannel, default=1.25, const [1.25,13]}
       Scale.Z {type=MovieSceneScriptingDoubleChannel, default=1.25, const [1.25,13]}
      ]
     }]
   }]
 }
 SM_Int_Boardgame_Dice_AG_FE {
  id=#2
  possessed_class=FortStaticMeshActor
  tracks [{
    type=MovieScene3DTransformTrack
    sections [{
      type=MovieScene3DTransformSection
      channels [
       Location.X {
        type=MovieSceneScriptingDoubleChannel
        default=2283.0896
        thinned_from=6
        keys [[0,2283.0896],[9.3338,2270.3877],[13.3338,2285.8101],[24.3338,2297.4402],[28.3338,2305.1748]]
       }
       Location.Y {type=MovieSceneScriptingDoubleChannel, default=-106942.0156, thinned_from=6, keys [[0,-106942.0156],[28.3338,-106920.7109]]}
       Location.Z {
        type=MovieSceneScriptingDoubleChannel
        default=16.5612
        keys [[0,16.5612],[5.3338,20.7946],[9.3338,20.2929],[13.3338,17.2107],[19.3338,21.3272],[24.3338,18.1894]]
       }
       Rotation.X {type=MovieSceneScriptingDoubleChannel, default=0, keys [[0.6288,251.3472],[8.77,124.2063],[28.2025,-222.3256]]}
       Rotation.Y {type=MovieSceneScriptingDoubleChannel, default=0, keys [[0,0],[9.2812,-98.784],[13.3787,-14.0864],[28.2025,-361.2357]]}
       Rotation.Z {type=MovieSceneScriptingDoubleChannel, default=0, keys [[0,0],[24.3338,0],[28.2025,145.2772]]}
       Scale.X {type=MovieSceneScriptingDoubleChannel, default=1.5, const [1.5,2]}
       Scale.Y {type=MovieSceneScriptingDoubleChannel, default=1.5, const [1.5,2]}
       Scale.Z {type=MovieSceneScriptingDoubleChannel, default=1.5, const [1.5,2]}
      ]
     }]
   }]
 }
]
display_rate {denominator=1, numerator=30}
tick_resolution {denominator=1, numerator=24000}
tracks [{
  type=MovieSceneAudioTrack
  display_name=Audio
  sections [{
    type=MovieSceneAudioSection
    end_frame=114
    start_frame=-7
    channels [
     Volume {type=MovieSceneScriptingFloatChannel, default=1}
     Pitch {type=MovieSceneScriptingFloatChannel, default=1}
     AttachActor {type=MovieSceneScriptingActorReferenceChannel, default="<Struct 'MovieSceneObjectBindingID' (0x851962bb3d0) {}>"}
    ]
   }]
 }]
